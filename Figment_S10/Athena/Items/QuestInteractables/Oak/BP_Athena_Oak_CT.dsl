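; Oak quest interactable that reveals an eye on a skeletal mesh by zeroing a dynamic material parameter and unhiding a flare.
; On non-authority clients, sphere overlap enables interaction when step 1's objective is done within an unfinished quest, or party assist applies to step 2.
; The local player's interaction disables further use and shows the eye, as does completed step 2 on playlist load.
; BeginPlay also calls ShowEye unconditionally.
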(event EventBeginPlay
  (CallFunction|ShowEye))

(event Custom|PlaylistLoaded
  (params (PlaylistName :type Name) (PlaylistContextTags :type Gameplay%20Tag%20Container%20Structure%20%28by%20ref%29))
  (bind _cancompletequest (Class|AthenaQuestFunctionLibrary|AthenaCheckCreativeMode (Game|GetGameState)))
  (if (and _cancompletequest (Utilities|Name|NotEqual(Name) (Variables|ReadOnly|GetStep2BackendName) "None"))
    (Networking|SwitchHasAuthority
      (:Remote
        (bind (_questisvalid _questcompleted) (Default|AthenaCheckQuestandObjectiveCompletion (Utilities|Array|Get(acopy) (Fortnite|GetLocalFortPlayerControllers true)) 0 (Variables|ReadOnly|GetStep2BackendName)))
        (if (and _questisvalid _questcompleted)
          (CallFunction|ShowEye))))))

(event Interaction|EventBlueprintOnInteract
  (params (InteractingPawn :type Fort%20Pawn%20Object%20Reference) (InteractionBeingAttempted :type EInteractionBeingAttempted%20Enum :default "FirstInteraction"))
  (CallFunction|PlayerInteractSuccess (Fortnite|GetFortPlayerControllerfromActor InteractingPawn)))

(event Custom|PlayerInteractSuccess
  (params (Player%20Controller :type Fort%20Player%20Controller%20Object%20Reference))
  (Networking|SwitchHasAuthority
    (:Remote
      (if (Utilities|Equal(Object) Player%20Controller (Utilities|Array|Get(acopy) (Fortnite|GetLocalFortPlayerControllers true)))
        (Variables|Default|SetCanInteract false)
        (CallFunction|ShowEye)))))

(event OnComponentBeginOverlap(Sphere)
  (params (OverlappedComponent :type Primitive%20Component%20Object%20Reference) (OtherActor :type Actor%20Object%20Reference) (OtherComp :type Primitive%20Component%20Object%20Reference) (OtherBodyIndex :type Integer) (bFromSweep :type Boolean) (SweepResult :type Hit%20Result%20Structure%20%28by%20ref%29))
  (Utilities|Casting|CastToPawn OtherActor)
  (Networking|SwitchHasAuthority
    (:Remote
      (bind _cancompletequest (Class|AthenaQuestFunctionLibrary|AthenaCheckCreativeMode (Game|GetGameState)))
      (if (and _cancompletequest (Utilities|Name|NotEqual(Name) (Variables|ReadOnly|GetStep1BackendName) "None"))
        (bind (_bisassistingobjective _bisvalidplayercontroller) (Class|AthenaQuestFunctionLibrary|AthenaPartyAssistCheck (Fortnite|GetFirstLocalFortPlayerController true) (Variables|ReadOnly|GetStep2BackendName)))
        ; Remove Eye Quest Def
        (bind (_questisvalid _questcompleted _objectivecompleted) (Default|AthenaCheckQuestandObjectiveCompletion (Fortnite|GetFirstLocalFortPlayerController true) 0 (Variables|ReadOnly|GetStep1BackendName)))
        (if (or (Math|Boolean|ANDBoolean _questisvalid (not _questcompleted) _objectivecompleted) (and _bisassistingobjective _bisvalidplayercontroller))
          (Variables|Default|SetCanInteract true))))))

(fn BlueprintCanInteract
  (params (InteractingPawn :type Fort%20Pawn%20Object%20Reference) (InteractionBeingAttempted :type EInteractionBeingAttempted%20Enum :default "FirstInteraction") (InteractionType :type TInteractionType%20Enum :default "IT_NoInteraction"))
  (return (Variables|Default|GetCanInteract)))

(fn BlueprintGetInteractionString
  (params (InteractingPawn :type Fort%20Pawn%20Object%20Reference) (InteractionBeingAttempted :type EInteractionBeingAttempted%20Enum :default "FirstInteraction"))
  (return (Variables|ReadOnly|GetInteractString)))

(fn ShowEye
  (Rendering|Material|SetScalarParameterValue (Variables|Default|GetDynamicMaterialInstance) (Variables|ReadOnly|GetMaterialParam) 0.0 :!class /Script/Engine.MaterialInstanceDynamic)
  (Development|SetHiddeninGame (Variables|Default|GetEyeFlare) false false))

(fn ConstructionScript
  (bind _returnvalue (Rendering|Material|CreateDynamicMaterialInstance (Variables|Default|GetSkeletalMesh) 0 (Variables|ReadOnly|GetBaseMaterialInstance) "None" :!class /Script/Engine.PrimitiveComponent))
  (bind _output_get (Variables|Default|SetDynamicMaterialInstance _returnvalue))
  (Rendering|Material|SetScalarParameterValue _output_get (Variables|ReadOnly|GetMaterialParam) 1.0 :!class /Script/Engine.MaterialInstanceDynamic))
