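; Quest interactable eye (Oak) that players can collect, shown only to local clients with the matching unfinished or party-shared objective.
; Server always allows interaction; clients check quest state after playlist load, toggling mesh and interactability.
; Overlap spheres drive timelines that raise the eye, bob it, periodically turn it toward the local player with servo sounds, and drop it on exit.
; Interacting, or a shared quest completing, hides it locally.
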(event EventBeginPlay
  (Networking|SwitchHasAuthority
    (:Authority
      (Variables|Default|SetCanInteract true))
    (:Remote
      (Utilities|FlowControl|Delay 5.0)
      (bind _asfort_game_state_athena (Utilities|Casting|CastToFortGameStateAthena (Game|GetGameState)))
      (if (Variables|/Script/FortniteGame.FortGameStateAthena|GetbPlaylistDataIsLoaded _asfort_game_state_athena)
        (bind _cancompletequest (Class|AthenaQuestFunctionLibrary|AthenaCheckCreativeMode (Game|GetGameState)))
        (if (and _cancompletequest (Utilities|Name|NotEqual(Name) (Variables|ReadOnly|GetBackendName) "None"))
          (Networking|SwitchHasAuthority
            (:Remote
              (bind (_bisassistingobjective _bisvalidplayercontroller) (Class|AthenaQuestFunctionLibrary|AthenaPartyAssistCheck (Fortnite|GetFirstLocalFortPlayerController true) (Variables|ReadOnly|GetBackendName)))
              (Variables|Default|SetbIsQuestShared (and _bisassistingobjective _bisvalidplayercontroller))
              ; Removed Fetch Eye Item Ref
              (bind (_questisvalid _questcompleted _objectivecompleted) (Default|AthenaCheckQuestandObjectiveCompletion (Fortnite|GetFirstLocalFortPlayerController true) 0 (Variables|ReadOnly|GetBackendName)))
              (bind _output_get (Variables|Default|SetCanInteract (or (Math|Boolean|ANDBoolean _questisvalid (not _questcompleted) (not _objectivecompleted)) (Variables|Default|GetbIsQuestShared))))
              (Development|SetHiddeninGame (Variables|Default|GetEyeSM) (not _output_get) false)
              (Utilities|FlowControl|Sequence
                (:then_0
                  ;[1 1 1] Bind enter and exit vols on the client
                  (if _output_get
                    (Collision|BindEventtoOnComponentBeginOverlap (Variables|Default|GetSphereOverlap) (EventDispatchers|CreateEvent :!event Custom|PlayerEntersVicinity))
                    (Collision|BindEventtoOnComponentEndOverlap (Variables|Default|GetSphereOverlapOuter) (EventDispatchers|CreateEvent :!event Custom|PlayerExitVicinity))))
                (:then_1
                  (if (Variables|Default|GetbIsQuestShared)
                    (bind _asfort_player_controller_athena (Utilities|Casting|CastToFortPlayerControllerAthena (Fortnite|GetFirstLocalFortPlayerController true)))
                    (Quests|BindEventtoOnQuestsUpdated (Quests|GetQuestManager _asfort_player_controller_athena "Invalid") (EventDispatchers|CreateEvent :!event Custom|ParentQuestUpdated))))))))
        (else
          (Playlist|BindEventtoOnCurrentPlaylistLoaded _asfort_game_state_athena (delegate Custom|PlaylistLoaded)))))))

(event Custom|PlaylistLoaded
  (params (PlaylistName :type Name) (PlaylistContextTags :type Gameplay%20Tag%20Container%20Structure%20%28by%20ref%29))
  (bind _cancompletequest (Class|AthenaQuestFunctionLibrary|AthenaCheckCreativeMode (Game|GetGameState)))
  (if (and _cancompletequest (Utilities|Name|NotEqual(Name) (Variables|ReadOnly|GetBackendName) "None"))
    (Networking|SwitchHasAuthority
      (:Remote
        (bind (_bisassistingobjective _bisvalidplayercontroller) (Class|AthenaQuestFunctionLibrary|AthenaPartyAssistCheck (Fortnite|GetFirstLocalFortPlayerController true) (Variables|ReadOnly|GetBackendName)))
        (Variables|Default|SetbIsQuestShared (and _bisassistingobjective _bisvalidplayercontroller))
        ; Removed Fetch Eye Item Ref
        (bind (_questisvalid _questcompleted _objectivecompleted) (Default|AthenaCheckQuestandObjectiveCompletion (Fortnite|GetFirstLocalFortPlayerController true) 0 (Variables|ReadOnly|GetBackendName)))
        (bind _output_get (Variables|Default|SetCanInteract (or (Math|Boolean|ANDBoolean _questisvalid (not _questcompleted) (not _objectivecompleted)) (Variables|Default|GetbIsQuestShared))))
        (Development|SetHiddeninGame (Variables|Default|GetEyeSM) (not _output_get) false)
        (Utilities|FlowControl|Sequence
          (:then_0
            ;[1 1 1] Bind enter and exit vols on the client
            (if _output_get
              (Collision|BindEventtoOnComponentBeginOverlap (Variables|Default|GetSphereOverlap) (EventDispatchers|CreateEvent :!event Custom|PlayerEntersVicinity))
              (Collision|BindEventtoOnComponentEndOverlap (Variables|Default|GetSphereOverlapOuter) (EventDispatchers|CreateEvent :!event Custom|PlayerExitVicinity))))
          (:then_1
            (if (Variables|Default|GetbIsQuestShared)
              (bind _asfort_player_controller_athena (Utilities|Casting|CastToFortPlayerControllerAthena (Fortnite|GetFirstLocalFortPlayerController true)))
              (Quests|BindEventtoOnQuestsUpdated (Quests|GetQuestManager _asfort_player_controller_athena "Invalid") (EventDispatchers|CreateEvent :!event Custom|ParentQuestUpdated)))))))))

(event Interaction|EventBlueprintOnInteract
  (params (InteractingPawn :type Fort%20Pawn%20Object%20Reference) (InteractionBeingAttempted :type EInteractionBeingAttempted%20Enum :default "FirstInteraction"))
  (CallFunction|PlayerInteractSuccess (Fortnite|GetFortPlayerControllerfromActor InteractingPawn)))

(event Custom|PlayerInteractSuccess
  (params (Player%20Controller :type Fort%20Player%20Controller%20Object%20Reference))
  (Networking|SwitchHasAuthority
    (:Remote
      (if (Utilities|Equal(Object) Player%20Controller (Fortnite|GetFirstLocalFortPlayerController true))
        (Variables|Default|SetCanInteract false)
        (Development|SetHiddeninGame (Variables|Default|GetEyeSM) true false)))))

(event Custom|PlayerEntersVicinity
  (params (OverlappedComponent :type Primitive%20Component%20Object%20Reference) (OtherActor :type Actor%20Object%20Reference) (OtherComp :type Primitive%20Component%20Object%20Reference) (OtherBodyIndex :type Integer) (bFromSweep :type Boolean) (SweepResult :type Hit%20Result%20Structure%20%28by%20ref%29))
  (Audio|Components|Audio|Play (Variables|Default|GetBBQ_Larry_EyeFloatLoop_Cue) 0.0)
  ;[1 1 1] Rise up
  (bind (_direction _raise_up_time) (Timeline|Timeline_0 PlayFromStart)
    (:Update
      (Transformation|SetRelativeLocation (Variables|Default|GetEyeSM) (Math|Vector|Lerp(Vector) "0, 0, 0" "0, 0,64.000000" _raise_up_time) false false))
    (:Finished
      (bind _returnvalue (Utilities|Time|SetTimerbyEvent (delegate Custom|RotateToLookAtPlayer) (Variables|Default|GetTimeBetweenRotations) true))
      (Variables|Default|SetRotationDelayTimer _returnvalue)
      (Audio|PlaySoundatLocation (Variables|Default|GetServoSounds) (Transformation|GetActorLocation))
      (Utilities|FlowControl|Sequence
        (:then_0
          ;[1 1 1] Bobbing motion
          (bind (_direction_2 _bob_time) (Timeline|Bob PlayFromStart)
            (:Update
              (Transformation|SetRelativeLocation (Variables|Default|GetEyeSM) (Math|Vector|Lerp(Vector) "0, 0,64.000000" "0, 0,74.000000" _bob_time) false false))))
        (:then_1
          ;[1 1 1] Rotate to Look at player
          (Variables|Default|SetCurrentRot (Variables|/Script/Engine.SceneComponent|GetRelativeRotation (Variables|Default|GetEyeSM)))
          (bind (_direction_3 _rotation_time) (Timeline|RotateToLook PlayFromStart)
            (:Update
              (Transformation|SetRelativeRotation (Variables|Default|GetEyeSM) (Math|Rotator|Lerp(Rotator) (Variables|Default|GetCurrentRot) (Math|Rotator|MakeRotator (.roll (Math|Rotator|MakeRotfromY (- (Transformation|GetActorLocation (Pawn|GetControlledPawn (Utilities|Array|Get(acopy) (Fortnite|GetLocalFortPlayerControllers true)))) (Transformation|GetActorLocation)))) (+ (.pitch (Math|Rotator|MakeRotfromY (- (Transformation|GetActorLocation (Pawn|GetControlledPawn (Utilities|Array|Get(acopy) (Fortnite|GetLocalFortPlayerControllers true)))) (Transformation|GetActorLocation)))) 1.8e+02) (.yaw (Math|Rotator|MakeRotfromY (- (Transformation|GetActorLocation (Pawn|GetControlledPawn (Utilities|Array|Get(acopy) (Fortnite|GetLocalFortPlayerControllers true)))) (Transformation|GetActorLocation))))) _rotation_time true) false false))))))))

(event Custom|ParentQuestUpdated
  (Networking|SwitchHasAuthority
    (:Remote
      ; Remove Quest Def for Eye
      (bind (_questisvalid _questcompleted _objectivecompleted) (Default|AthenaCheckQuestandObjectiveCompletion (Fortnite|GetFirstLocalFortPlayerController true) 0 (Variables|ReadOnly|GetBackendName)))
      (if (or _questcompleted _objectivecompleted)
        (Variables|Default|SetCanInteract false)
        (Development|SetHiddeninGame (Variables|Default|GetEyeSM) true false)
        (bind _asfort_player_controller_athena (Utilities|Casting|CastToFortPlayerControllerAthena (Fortnite|GetFirstLocalFortPlayerController true)))
        (Quests|UnbindEventfromOnQuestsUpdated (Quests|GetQuestManager _asfort_player_controller_athena "Invalid") (EventDispatchers|CreateEvent :!event Custom|ParentQuestUpdated))))))

(event Custom|RotateToLookAtPlayer
  (Audio|PlaySoundatLocation (Variables|Default|GetServoSounds) (Transformation|GetActorLocation))
  (Utilities|FlowControl|Sequence
    (:then_0
      ;[1 1 1] Bobbing motion
      (Timeline|Bob PlayFromStart))
    (:then_1
      ;[1 1 1] Rotate to Look at player
      (Variables|Default|SetCurrentRot (Variables|/Script/Engine.SceneComponent|GetRelativeRotation (Variables|Default|GetEyeSM)))
      (Timeline|RotateToLook PlayFromStart))))

(event Custom|PlayerExitVicinity
  (params (OverlappedComponent :type Primitive%20Component%20Object%20Reference) (OtherActor :type Actor%20Object%20Reference) (OtherComp :type Primitive%20Component%20Object%20Reference) (OtherBodyIndex :type Integer))
  (bind _eyesm (Variables|Default|GetEyeSM))
  (bind _returnvalue (Transformation|GetRelativeTransform _eyesm))
  (bind _eyesm_2 (Variables|Default|GetEyeSM))
  (Utilities|FlowControl|Sequence
    (:then_0
      ;[1 1 1] Tell other timelines to stop
      (Components|Timeline|Stop (Variables|BP_Athena_Oak_CT_Eye|GetBob))
      (Components|Timeline|Stop (Variables|BP_Athena_Oak_CT_Eye|GetRotateToLook))
      (Components|Timeline|Stop (Variables|BP_Athena_Oak_CT_Eye|GetTimeline_0))
      (Audio|Components|Audio|Stop (Variables|Default|GetBBQ_Larry_EyeFloatLoop_Cue)))
    (:then_1
      ;[1 1 1] Drop the eye back down
      (Utilities|Time|ClearandInvalidateTimerbyHandle (Variables|Default|GetRotationDelayTimer))
      (Variables|Default|SetDropStartLoc (.location _returnvalue))
      (Variables|Default|SetDropStartRot (.rotation _returnvalue))
      (bind (_direction _droptime) (Timeline|Drop PlayFromStart)
        (:Update
          (Transformation|SetRelativeLocationAndRotation _eyesm_2 (Math|Vector|Lerp(Vector) (Variables|Default|GetDropStartLoc) "0, 0,10.000000" _droptime) (Math|Rotator|Lerp(Rotator) (Variables|Default|GetDropStartRot) "0, 0, 0" _droptime) false false))
        (:Finished
          (Transformation|SetRelativeRotation _eyesm_2 "0, 0, 0" false false))))))

(fn BlueprintCanInteract
  (params (InteractingPawn :type Fort%20Pawn%20Object%20Reference) (InteractionBeingAttempted :type EInteractionBeingAttempted%20Enum :default "FirstInteraction") (InteractionType :type TInteractionType%20Enum :default "IT_NoInteraction"))
  (return (Variables|Default|GetCanInteract)))

(fn BlueprintGetInteractionString
  (params (InteractingPawn :type Fort%20Pawn%20Object%20Reference) (InteractionBeingAttempted :type EInteractionBeingAttempted%20Enum :default "FirstInteraction"))
  (return (Variables|ReadOnly|GetInteractString)))

(fn ConstructionScript
  (Variables|Default|SetNeVar_0 (Variables|/Script/Engine.SceneComponent|GetRelativeLocation (Variables|Default|GetEyeSM))))
