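; Gameplay ability GA_AthenaOak_DropCash, from the Oak quest interactables, that drops a cash pickup in the world.
; When activated by a gameplay event, it reads the event's target and spawns one cash item at that actor's location, tossed with an upward direction.
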(event Ability|EventActivateAbilityFromEvent
  (params (EventData :type Gameplay%20Event%20Data%20Structure%20%28by%20ref%29))
  (bind (_eventtag _instigator _target) (Utilities|Struct|BreakGameplayEventData EventData))
  (Inventory|SpawnPickupInWorld (Variables|Default|GetCashItem) 1 (Transformation|GetActorLocation _target) "0.000000,0.000000,1.000000" 0 true true false -1 "Tossed" "Unset" 0 false))
